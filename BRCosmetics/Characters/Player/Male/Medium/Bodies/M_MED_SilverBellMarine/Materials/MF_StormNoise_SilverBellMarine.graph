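# Material function that builds animated, colored storm noise for the SilverBellMarine outfit, with a separate scalar noise output.
# A panning distortion texture warps a second scrolling noise sample, projected either by UVs or triplanar with parallax; the triplanar path perturbs tiling size.
# The red channel drives a dark-to-light gradient and green a tint, both brightness-boosted.
# It also derives a sunlight wave term and a noise-normal fresnel.
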MaterialFunction /BRCosmetics/Characters/Player/Male/Medium/Bodies/M_MED_SilverBellMarine/Materials/MF_StormNoise_SilverBellMarine
graph {nodes [
  FunctionOutput(None=NamedRerouteUsage())
  n1 = NamedRerouteUsage()
  n2 = NamedRerouteUsage()
  n3 = Add(A=Saturate(None=Add(A=@n1, B=@n2)), B=NamedRerouteUsage())
  FunctionOutput(None=NamedRerouteDeclaration(None=Saturate(None=@n3)) {name=Noise_OUT, nodeColor=<rgba 1 0.5156 0 1>}) {outputName=Noise, sortPriority=1}
  n4 = TextureObject() {
   samplerType=SAMPLERTYPE_Normal
   texture=/BRCosmetics/Characters/Player/Female/Medium/Bodies/F_MED_SilverBellWind/Textures/T_WaterDistortion_N
  }
  MaterialFunctionCall!(TextureObject=@n4, TextureSize, WorldPosition, Export Float 4, In, World Space Normal, ProjectionTransitionContrast) {materialFunction=/Game/Athena/SafeZone/WorldAlignedTexture_ScalePan}
  n5 = ScalarParameter() {parameterName=Parallax}
  n6 = VectorParameter() {group=StormNoise, parameterName=DistortionSpeed, channelNames {a="", b="", g=Speed Y, r=Speed X}, defaultValue=<rgba -0.1 0.1 1 1>}
  n7 = ComponentMask(None=@n6) {g=true, r=true}
  n8 = TextureCoordinate() {coordinateIndex=1}
  n9 = TextureCoordinate()
  n10 = NamedRerouteDeclaration(None=FunctionInput(Preview=StaticSwitchParameter(True=@n8, False=@n9) {group=StormNoise, parameterName=UseUVchannel01}) {bUsePreviewValueAsDefault=true, inputName=UV, inputType=FunctionInput_Vector2}) {name=UVs, nodeColor=<rgba 0 0.5078 1 1>}
  n11 = VectorParameter() {group=StormNoise, parameterName=DistortionTiling, channelNames {a="", b="", g=Tile Y, r=Tile X}, defaultValue=<rgba 1 1 1 1>}
  n12 = ComponentMask(None=@n11) {g=true, r=true}
  n13 = TextureSampleParameter2D(UVs=Panner(Coordinate=Multiply(A=@n10, B=@n12), Speed=@n7)) {
   group=StormNoise
   parameterName=DistortionTexture
   samplerSource=SSM_Wrap_WorldGroupSettings
   samplerType=SAMPLERTYPE_LinearColor
   texture=&T_NoisePack_LinearColor
  }
  n14 = StaticSwitch(
   True=MaterialFunctionCall!(
    TextureObject=TextureObject() {samplerType=SAMPLERTYPE_LinearColor, texture=&T_NoisePack_LinearColor}
    TextureSize=VectorParameter() {group=StormNoise, parameterName=DistortionTiling_Triplanar, defaultValue=<rgba 80 80 80 1>}.RGB
    WorldPosition=NamedRerouteUsage()
    Export Float 4
    Parallax=NamedRerouteDeclaration(None=@n5) {name=Parallax, nodeColor=<rgba 0 1 0.7891 1>}
    In=@n7
    World Space Normal=NamedRerouteUsage()
    ProjectionTransitionContrast=ScalarParameter() {parameterName=TriplanarProjContrast}
   ) {materialFunction=&/MF_Triplanar_ScalePan_Parallax}.XYZ Texture
   False=@n13.RGB
   Value=NamedRerouteUsage()
  )
  NamedRerouteDeclaration(None=ChannelMaskParameter(None=@n14) {group=StormNoise, parameterName=DistortionMaskChannel}.R) {name=DistortionMask, nodeColor=<rgba 0 1 0.0859 1>}
  n15 = ConstantBiasScale(None=@n14) {bias=-0.5, scale=2}
  NamedRerouteDeclaration(None=ComponentMask(None=@n15) {g=true, r=true}) {name=UVDistortion, nodeColor=<rgba 1 0 0.4922 1>}
  n16 = FunctionInput(Preview=GetMaterialAttributes().MaterialAttributes) {bUsePreviewValueAsDefault=true, inputName=Material, inputType=FunctionInput_MaterialAttributes}
  NamedRerouteDeclaration(None=GetMaterialAttributes(None=NamedRerouteDeclaration(None=@n16) {name=InMaterial, nodeColor=<rgba 1 0.9375 0 1>}) {attributeGetTypes [#1]}.Normal) {name=InNormal, nodeColor=<rgba 0.7578 0 1 1>}
  n17 = CollectionParameter() {
   collection=&FortniteMaterialParameters
   parameterName=SunlightDirectionalVector
  }
  n18 = ComponentMask(None=NamedRerouteDeclaration(None=@n17) {name=LightVecWS, nodeColor=<rgba 0 1 0.9297 1>}) {b=true, g=true, r=true}
  n19 = MaterialFunctionCall(Float3=@n18) {materialFunction=&BreakOutFloat3Components}
  n20 = MaterialFunctionCall(X=@n19.R, Y=@n19.R, Z=Max(A=@n19.B) {constB=0.2}) {materialFunction=&MakeFloat3}
  n21 = MaterialFunctionCall(Float3=@n18) {materialFunction=&BreakOutFloat3Components}
  n22 = Normalize(VectorInput=MaterialFunctionCall(X=@n21.R, Y=@n21.R, Z=Max(A=@n21.B) {constB=0.9}) {materialFunction=&MakeFloat3}.Result)
  n23 = NamedRerouteUsage()
  n24 = DotProduct(A=Normalize(VectorInput=LinearInterpolate(A=@n20.Result, B=@n22, Alpha=@n23)), B=Transform(None=NamedRerouteUsage()))
  NamedRerouteDeclaration(None=Power(Base=Saturate(None=@n24), Exp=ScalarParameter() {defaultValue=5, parameterName=LightPower})) {name=LightWave, nodeColor=<rgba 0 1 0.5078 1>}
  NamedRerouteDeclaration(None=FunctionInput(Preview=StaticBool() {value=true}) {bUsePreviewValueAsDefault=true, inputName=UseTriplanar, inputType=FunctionInput_StaticBool}) {name=UseTriplanar, nodeColor=<rgba 0 0.9297 1 1>}
  NamedRerouteDeclaration(None=FunctionInput() {bUsePreviewValueAsDefault=true, inputName=FrontEndOverride, inputType=FunctionInput_Scalar, sortPriority=3}) {name=FrontEndOverride, nodeColor=<rgba 0 0.7187 1 1>}
  NamedRerouteDeclaration(None=VertexInterpolator(VS=PreSkinnedPosition()).PS) {name=PS-LocalPosition, nodeColor=<rgba 1 0 0.6562 1>}
  NamedRerouteDeclaration(None=VertexInterpolator(VS=PreSkinnedNormal()).PS) {name=PS-LocalNormal, nodeColor=<rgba 0 1 0.7656 1>}
  NamedRerouteDeclaration(None=@n14) {name=Debug, nodeColor=<rgba 0.875 1 0 1>}
  n25 = TextureObject() {
   samplerType=SAMPLERTYPE_Masks
   texture=/BRCosmetics/Characters/Player/Male/Medium/Bodies/M_MED_SilverBellMarine/Skins/Dark/Textures/T_SilverBellMarine_Dark_Body_TA
  }
  n26 = ScalarParameter() {defaultValue=2, group=StormNoise, parameterName=DistortionStrength_Triplanar}
  n27 = VectorParameter() {group=StormNoise, parameterName=NoiseSpeed, channelNames {a="", b="", g=Speed Y, r=Speed X}, defaultValue=<rgba 0.025 -0.05 1 1>}
  n28 = ComponentMask(None=@n27) {g=true, r=true}
  n29 = MaterialFunctionCall(
   TextureObject=@n25
   TextureSize=Add(A=Multiply(A=@n15, B=@n26), B=VectorParameter() {group=StormNoise, parameterName=NoiseTiling_Triplanar, defaultValue=<rgba 80 80 80 1>}.RGB)
   WorldPosition=NamedRerouteUsage()
   Parallax=NamedRerouteUsage()
   In=@n28
   World Space Normal=NamedRerouteUsage()
  ) {materialFunction=&MF_Triplanar_ScalePan_Parallax}
  n30 = VectorParameter() {group=StormNoise, parameterName=NoiseTiling, channelNames {a="", b="", g=Tile Y, r=Tile X}, defaultValue=<rgba 1 1 1 1>}
  n31 = Multiply(A=NamedRerouteUsage(), B=ComponentMask(None=@n30) {g=true, r=true})
  n32 = Multiply(A=ScalarParameter() {defaultValue=0.02, group=StormNoise, parameterName=DistortionStrength}, B=NamedRerouteUsage())
  n33 = Time()
  n34 = TextureSampleParameter2D(UVs=Add(A=Add(A=@n31, B=@n32), B=Multiply(A=@n33, B=@n28))) {
   group=StormNoise
   parameterName=NoiseTexture01
   samplerType=SAMPLERTYPE_Masks
   texture=&T_SilverBellMarine_Dark_Body_TA
  }
  n35 = NamedRerouteUsage()
  NamedRerouteDeclaration(None=StaticSwitch(True=ComponentMask(None=@n29.XYZ Texture) {r=true}, False=@n34.R, Value=@n35)) {name=NoiseSample, nodeColor=<rgba 0 0.3437 1 1>}
  n36 = TextureSampleParameter2D() {
   group=StormNoise
   parameterName=NoiseNormal
   samplerType=SAMPLERTYPE_Normal
   texture=&T_WaterDistortion_N
  }
  NamedRerouteDeclaration(None=StaticSwitchParameter(True=Constant3Vector() {constant=<rgba 0 0 1 0>}, False=@n36.RGB) {group=StormNoise, parameterName=UseTriplanar}) {name=NoiseNormalSample, nodeColor=<rgba 1 0 0.3516 1>}
  NamedRerouteDeclaration(None=StaticSwitch(True=ComponentMask(None=@n29.XYZ Texture) {g=true}, False=@n34.G, Value=@n35)) {name=NoiseSampleTint, nodeColor=<rgba 1 0 0.8437 1>}
  n37 = NamedRerouteUsage()
  n38 = OneMinus(None=DotProduct(A=CameraVectorWS(), B=Transform(None=@n37)))
  NamedRerouteDeclaration(None=Power(Base=ConstantBiasScale(None=@n38))) {name=NoiseNormalFresnel, nodeColor=<rgba 0 1 0.0859 1>}
  n39 = VectorParameter() {group=StormNoise, parameterName=NoiseColorDark, defaultValue=<rgba 0.0088 0.0392 0.0999 1>}
  n40 = VectorParameter() {group=StormNoise, parameterName=NoiseColorLight, defaultValue=<rgba 0.2646 0.5736 0.7011 1>}
  n41 = Saturate(None=Add(A=NamedRerouteUsage()) {constB=0})
  n42 = Power(Base=@n41, Exp=ScalarParameter() {defaultValue=1.3, group=StormNoise, parameterName=NoisePower01}) {constExponent=1.5}
  n43 = Multiply(A=@n41) {constB=0.9}
  n44 = ScalarParameter() {defaultValue=6, group=StormNoise, parameterName=HighPower01}
  n45 = Multiply(
   A=Multiply(A=LinearInterpolate(A=@n39.RGB, B=@n40.RGB, Alpha=@n42))
   B=LinearInterpolate(
    A=ScalarParameter() {defaultValue=1, group=StormNoise, parameterName=Brightness01}
    B=ScalarParameter() {defaultValue=3, group=StormNoise, parameterName=BrightnessHigh01}
    Alpha=Power(Base=@n43, Exp=@n44) {constExponent=4}
   )
  )
  n46 = NamedRerouteUsage()
  n47 = ScalarParameter() {defaultValue=6, group=StormNoise, parameterName=HighPower02}
  n48 = Multiply(
   A=Multiply(A=@n46, B=VectorParameter() {group=StormNoise, parameterName=NoiseTint, defaultValue=<rgba 1 0.1389 0.9332 1>}.RGB)
   B=LinearInterpolate(
    A=ScalarParameter() {defaultValue=2, group=StormNoise, parameterName=Brightness02}
    B=ScalarParameter() {defaultValue=5, group=StormNoise, parameterName=BrightnessHigh02}
    Alpha=Power(Base=@n46, Exp=@n47)
   )
  )
  NamedRerouteDeclaration(None=Multiply(A=Add(A=@n45, B=@n48), B=NamedRerouteUsage())) {name=Output, nodeColor=<rgba 0.0312 0 1 1>}
  n49 = FunctionInput() {bUsePreviewValueAsDefault=true, inputName=Alpha, inputType=FunctionInput_Scalar, sortPriority=1, previewValue=<1 0 0 1>}
  NamedRerouteDeclaration(None=@n49) {name=Alpha, nodeColor=<rgba 0.1484 0 1 1>}
  NamedRerouteUsage()
  NamedRerouteUsage()
  NamedRerouteUsage()
  Transform(None=FunctionInput(Preview=Constant3Vector() {constant=<rgba 0 0 1 0>}) {bUsePreviewValueAsDefault=true, inputName=Normal}) {transformType=TRANSFORM_Local}
 ]}
