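# Parent character material for the SilverBellMarine body, wrapping optional skin effects between the shared 2023 PreFX and PostFX character functions.
# Static switches gate Slurp, vertex-color index eyes, SilverBell energy and flipbook mouths, with eyes and mouth sharing a tint and glow.
# The energy is storm noise blended into emissive with inverted Fresnel, projected triplanar or on camera-angle-blended virtual planes that can follow the actor.
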Material /BRCosmetics/Characters/Player/Male/Medium/Bodies/M_MED_SilverBellMarine/Materials/M_SilverBellMarine_Parent
graph {
 properties {
  bUseMaterialAttributes=true
  bUsedWithClothing=true
  bUsedWithMeshDeformer=true
  bUsedWithMorphTargets=true
  bUsedWithNanite=true
  bUsedWithSkeletalMesh=true
  pixelDepthOffsetMode=PDOM_Legacy
  shadingModel=MSM_FromMaterialExpression
  usedShadingModels=Default Lit | Subsurface | Clear Coat | Subsurface Profile | Cloth
 }
 nodes [
  n1 = MaterialFunctionCall() {materialFunction=&MF_CharacterParent_2023_PreFX}
  MaterialFunctionCall(Material=NamedRerouteUsage(), M=@n1.M, Pre Skinned Local Position=@n1.M, Distance Blend=@n1.M, Fresnel=@n1.M) {materialFunction=&MF_CharacterParent_2023_PostFX}
  n2 = StaticBool()
  n3 = VectorParameter() {parameterName=VP_NormalAxis, defaultValue=<rgba 1 0 0 1>}
  n4 = ScalarParameter() {defaultValue=1, parameterName=UseActorPos}
  n5 = LinearInterpolate(A=@n3.RGB, B=Transform(None=@n3.RGB) {transformSourceType=TRANSFORMSOURCE_Local}, Alpha=@n4)
  n6 = VectorParameter() {parameterName=VP_XAxis, defaultValue=<rgba 0 1 0 1>}
  n7 = LinearInterpolate(A=@n6.RGB, B=Transform(None=@n6.RGB) {transformSourceType=TRANSFORMSOURCE_Local}, Alpha=@n4)
  n8 = LinearInterpolate(A=VectorParameter() {parameterName=VP_PlanePos, defaultValue=<rgba 0 0 0 1>}.RGB, B=ActorPositionWS(), Alpha=@n4)
  n9 = ScalarParameter() {defaultValue=512, parameterName=SilverBellEngergy_VP_Tiling}
  n10 = MaterialFunctionCall(Plane Normal Axis=@n5, Plane X Axis=@n7, Plane Center=@n8, UV Scale=@n9) {materialFunction=&VirtualPlaneCoordinates}
  n11 = NamedRerouteUsage()
  n12 = NamedRerouteUsage()
  n13 = MaterialFunctionCall(Plane Normal Axis=@n7, Plane X Axis=@n5, Plane Center=@n8, UV Scale=@n9) {materialFunction=&VirtualPlaneCoordinates}
  n14 = NamedRerouteUsage()
  n15 = NamedRerouteUsage()
  n16 = Abs(None=DotProduct(A=@n7, B=CameraVectorWS()))
  n17 = NamedRerouteDeclaration(None=MaterialFunctionCall(In=ComponentMask(None=@n16) {r=true}, Contrast=ScalarParameter() {defaultValue=0.5, parameterName=ContrastX}) {materialFunction=&CheapContrast}.Result) {name=BlendAlphaX, nodeColor=<rgba 0 1 0.7422 1>}
  n18 = ComponentMask(None=Abs(None=CameraVectorWS())) {b=true}
  n19 = ScalarParameter() {parameterName=ContrastZ}
  n20 = LinearInterpolate(
   A=LinearInterpolate(
    A=MaterialFunctionCall!(Normal, UseTriplanar=@n2, UV=@n10.UVs, Material=@n11, Alpha=@n12, FrontEndOverride) {materialFunction=&/MF_StormNoise_SilverBellMarine}.Result
    B=MaterialFunctionCall(UseTriplanar=@n2, UV=@n13.UVs, Material=@n14, Alpha=@n15) {materialFunction=&MF_StormNoise_SilverBellMarine}.Result
    Alpha=@n17
   ) {constAlpha=0, constB=0}
   Alpha=NamedRerouteDeclaration(None=MaterialFunctionCall(In=@n18, Contrast=@n19) {materialFunction=&CheapContrast}.Result) {name=BlendAlphaZ, nodeColor=<rgba 0 1 0.7422 1>}
  ) {constB=0}
  Multiply(A=@n20, B=NamedRerouteUsage())
  NamedRerouteDeclaration(None=@n20) {name=SilverBellEnergy_Color_VP, nodeColor=<rgba 0.0078 1 0 1>}
  n21 = NamedRerouteUsage()
  n22 = GetMaterialAttributes(None=MaterialFunctionCall!(Material=@n21) {materialFunction=/Game/Characters/Player/Common/MaterialFunctions/VertexColorFaces/MF_FV2_IndexEyes}.Result) {attributeGetTypes [#1, #2]}
  n23 = NamedRerouteUsage()
  n24 = VectorParameter() {group=SilverBellMarine, parameterName=SBM_MouthColor, defaultValue=<rgba 0.3542 0.3542 0.3542 1>}
  n25 = Multiply(A=@n24.RGB, B=ScalarParameter() {defaultValue=0.3, group=SilverBellMarine, parameterName=SBM_EmissiveIntensity})
  n26 = StaticSwitchParameter(
   True=SetMaterialAttributes!(
    MaterialAttributes=@n22.MaterialAttributes
    World Position Offset=LinearInterpolate(A=@n22.World Position Offset, Alpha=@n23) {constB=0}
    Emissive Color=LinearInterpolate(A=@n25, B=@n22.Emissive Color, Alpha=@n23)
   ) {attributeSetTypes [#1, #2]}
   False=@n21
  ) {group=Skin FX, parameterName=UseIndexEyes}
  n27 = GetMaterialAttributes(None=@n26) {attributeGetTypes [#2]}
  n28 = Fresnel() {baseReflectFraction=0, exponent=0.2}
  n29 = Multiply(
   A=StaticSwitch(True=NamedRerouteUsage(), False=NamedRerouteUsage(), Value=StaticBoolParameter() {parameterName=UseTriplanar})
   B=NamedRerouteDeclaration(None=OneMinus(None=@n28)) {name=Fresnel, nodeColor=<rgba 0 0.4609 1 1>}
  )
  n30 = Multiply(A=NamedRerouteUsage(), B=ScalarParameter() {defaultValue=0.15, parameterName=SilverBellEnergy_Intensity})
  n31 = StaticSwitchParameter(
   True=SetMaterialAttributes!(
    MaterialAttributes=NamedRerouteDeclaration(None=@n27.MaterialAttributes) {name=MA_1, nodeColor=<rgba 0.3594 0 1 1>}
    Emissive Color=LinearInterpolate(A=@n27.Emissive Color, B=@n29, Alpha=@n30)
   ) {attributeSetTypes [#2]}
   False=@n26
  ) {group=Skin FX, parameterName=UseSilverBellEnergy}
  n32 = MaterialFunctionCall!(Material=@n31, MouthUVs) {materialFunction=/Game/Characters/Player/Common/MaterialFunctions/VertexColorFaces/MF_FV2_FlipbookMouths}
  n33 = GetMaterialAttributes(None=@n32.Result) {attributeGetTypes [
    #3
    #2
    #4
    #5
   ]}
  n34 = SetMaterialAttributes!(
   MaterialAttributes=@n33.MaterialAttributes
   Base Color=LinearInterpolate(A=@n33.Base Color, B=@n24.RGB, Alpha=@n32.Alpha)
   Emissive Color=LinearInterpolate(A=@n33.Emissive Color, B=@n25, Alpha=@n32.Alpha)
   Specular=LinearInterpolate(A=@n33.Specular, B=Constant() {r=0.5}, Alpha=@n32.Alpha)
   Roughness=LinearInterpolate(A=@n33.Roughness, B=Constant() {r=0.8}, Alpha=@n32.Alpha)
  ) {attributeSetTypes [
    #3
    #2
    #4
    #5
   ]}
  NamedRerouteDeclaration(None=StaticSwitchParameter(True=@n34, False=@n31) {group=Skin FX, parameterName=UseFlipbookMouths}) {name=MA_OUT, nodeColor=<rgba 1 0.6094 0 1>}
  n35 = StaticComponentMaskParameter(None=@n1.Tech Art Mask) {defaultR=true, group=Skin FX, parameterName=Slurp Channel}
  n36 = ComponentMask(None=@n1.M) {r=true}
  n37 = StaticComponentMaskParameter(None=@n1.Tech Art Mask) {defaultG=true, group=Skin FX, parameterName=Slurp Bubbles Channel}
  n38 = StaticComponentMaskParameter(None=@n1.Tech Art Mask) {defaultB=true, group=Skin FX, parameterName=Slurp Normal Distortion Channel}
  NamedRerouteDeclaration(None=StaticSwitchParameter(
    True=MaterialFunctionCall!(
     Alpha=@n35
     AO=@n36
     BubblesMask=@n37
     SlurpNormalDistortionMask=@n38
     Pre Skinned Position=@n1.Pre skinned Local Position
     Material=@n1.Pre skinned Local Position
    ) {materialFunction=/Game/Weapons/WeaponSkins/MaterialFunction/MF_Slurp}.Result
    False=@n1.Result
   ) {group=Skin FX, parameterName=UseSlurp}) {name=MA_IN, nodeColor=<rgba 1 0.1406 0 1>}
  NamedRerouteDeclaration(None=MaterialFunctionCall(UseTriplanar=StaticBool() {value=true}, Material=NamedRerouteUsage(), Alpha=NamedRerouteUsage()) {materialFunction=&MF_StormNoise_SilverBellMarine}.Result) {name=SilverBellEnergy_Color_Triplanar, nodeColor=<rgba 0 0.1562 1 1>}
  NamedRerouteDeclaration(None=@n17) {name=Debug, nodeColor=<rgba 0 0.4844 1 1>}
  n39 = Subtract(A=PixelDepth().R, B=Distance(A=CameraPositionWS(), B=ObjectPositionWS()))
  n40 = ObjectRadius()
  n41 = Saturate(None=ConstantBiasScale(None=Divide(A=@n39, B=@n40) {constB=512}))
  n42 = Constant() {r=1}
  NamedRerouteDeclaration(None=QualitySwitch(Default=OneMinus(None=@n41), Low=@n42, Medium=@n42)) {name=CameraDepth, nodeColor=<rgba 0 0.9531 1 1>}
  NamedRerouteDeclaration(None=@n13.World Position) {nodeColor=<rgba 0.5 0 1 1>}
  NamedRerouteDeclaration(None=NamedRerouteDeclaration(None=@n35) {name=SlurpChannel, nodeColor=<rgba 0.9688 1 0 1>}) {name=EnergyChannel, nodeColor=<rgba 1 0 0.2088 1>}
  NamedRerouteUsage()
  StaticComponentMaskParameter(None=@n1.Tech Art Mask) {defaultR=true, group=Skin FX, parameterName=Energy Channel}
 ]
}
